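# Two-sided lit material for bird particles in the LazarusLens skydiving contrail, usable on Niagara meshes and ribbons.
# Vertex animation math builds sine-driven rotations about local Y, masked by vertex color channels, for wing-like motion, plus vertical bobbing and traveling waves.
# Dynamic parameters vary that motion per particle, with several alternative offset chains present.
# Color and normal textures supply the surface, with particle color and a two-color UV1 gradient tinting base and emissive.
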Material /BRCosmetics/Athena/Effects/Particles/Skydiving/NS_Contrail_LazarusLens/FX/M_Contrail_LazarusLens_Bird
graph {
 properties {
  bUsedWithNiagaraMeshParticles=true
  bUsedWithNiagaraRibbons=true
  pixelDepthOffsetMode=PDOM_Legacy
  twoSided=true
  usedShadingModels=Default Lit
 }
 nodes [
  n1 = Transform(None=Constant3Vector() {constant=<rgba 0 1 0 0>}) {transformSourceType=TRANSFORMSOURCE_Local}
  n2 = VertexColor()
  n3 = DynamicParameter() {defaultValue=<rgba 1 1 0 0>, paramNames [Bias,Speed,Param3,Param4]}
  n4 = Multiply(A=Add(A=Time(), B=@n3.Bias), B=@n3.Speed)
  n5 = Sine(None=Multiply(A=@n4, B=ScalarParameter() {defaultValue=1, parameterName=Speed}))
  n6 = ScalarParameter() {parameterName=Rotate Amount}
  n7 = Multiply(A=Subtract(A=Multiply(A=@n2.R) {constB=2}), B=Add(A=Multiply(A=@n5, B=@n6), B=ScalarParameter() {parameterName=Rotation Angle Bias}))
  n8 = ObjectPositionWS()
  n9 = WorldPosition()
  n10 = Transform(None=Constant3Vector() {constant=<rgba 0 1 0 0>}) {transformSourceType=TRANSFORMSOURCE_Local}
  n11 = VertexColor()
  n12 = Multiply(A=Add(A=ScalarParameter() {parameterName=Forward Rotate Delay}, B=@n4), B=ScalarParameter() {defaultValue=1, parameterName=Speed})
  n13 = Multiply(A=Subtract(A=Multiply(A=@n11.G) {constB=2}), B=Multiply(A=Sine(None=@n12), B=ScalarParameter() {parameterName=Forward rotate Amount}))
  n14 = ObjectPositionWS()
  n15 = WorldPosition()
  n16 = Multiply(
   A=Add(
    A=RotateAboutAxis(NormalizedRotationAxis=@n1, RotationAngle=@n7, PivotPoint=@n8, Position=@n9.XYZ)
    B=RotateAboutAxis(NormalizedRotationAxis=@n10, RotationAngle=@n13, PivotPoint=@n14, Position=@n15.XYZ)
   )
   B=VertexColor().A
  )
  n17 = DynamicParameter() {defaultValue=<rgba 1 1 0 0>, paramNames [Bias,Speed,Param3,Param4]}
  n18 = Sine(None=Multiply(A=@n17.Param4, B=Time())) {period=3}
  n19 = Constant3Vector() {constant=<rgba 0 0 1 0>}
  Add(A=@n16, B=Multiply(A=@n17.Param3, B=Multiply(A=@n18, B=@n19)))
  n20 = Constant3Vector() {constant=<rgba 0 0 1 0>}
  n21 = Multiply(A=Add(A=@n4, B=ScalarParameter() {parameterName=Macro Time Bias}), B=ScalarParameter() {defaultValue=1, parameterName=Speed})
  n22 = WorldPosition()
  n23 = VertexColor()
  n24 = ScalarParameter() {parameterName=UseMacroWingFreq?}
  n25 = Divide(
   A=ComponentMask(None=Transform(None=@n22.XYZ) {transformSourceType=TRANSFORMSOURCE_World, transformType=TRANSFORM_Local}) {r=true}
   B=LinearInterpolate(
    A=ScalarParameter() {defaultValue=128, parameterName=Macro Wave Frequency}
    B=ScalarParameter() {defaultValue=128, parameterName=Macro Wave Frequency Wings}
    Alpha=Multiply(A=@n23.A, B=@n24)
   )
  )
  n26 = Multiply(A=Sine(None=Add(A=@n21, B=@n25)), B=ScalarParameter() {defaultValue=1, parameterName=Macro Wave scale})
  n27 = VertexColor()
  Add(A=Multiply(A=Transform(None=@n20) {transformSourceType=TRANSFORMSOURCE_Local}, B=Multiply(A=@n26, B=@n27.B)), B=@n16)
  n28 = Normalize(VectorInput=Transform(None=Constant3Vector() {constant=<rgba 0 1 0 0>}) {transformSourceType=TRANSFORMSOURCE_Local})
  n29 = Constant() {r=0.08}
  n30 = Time()
  n31 = DynamicParameter() {defaultValue=<rgba 1 1 0 0>, paramNames [Bias,Speed,Param3,Param4]}
  n32 = Add(A=Sine(None=Multiply(A=@n30, B=@n31.Speed)), B=@n31.Param4) {constB=-0.35}
  n33 = ComponentMask(None=TextureCoordinate()) {r=true}
  n34 = Multiply(A=Multiply(A=Multiply(A=@n29, B=@n32), B=Add(A=@n33, B=@n31.Param3) {constB=0.5}), B=@n31.Bias)
  n35 = ObjectPositionWS()
  n36 = WorldPosition()
  n37 = Normalize(VectorInput=Transform(None=Constant3Vector() {constant=<rgba 0 1 0 0>}) {transformSourceType=TRANSFORMSOURCE_Local})
  n38 = Constant() {r=-0.08}
  n39 = Time()
  n40 = Add(A=Sine(None=Multiply(A=@n39, B=@n31.Speed) {constB=5}), B=@n31.Param4) {constB=-0.35}
  n41 = ComponentMask(None=TextureCoordinate()) {r=true}
  n42 = Multiply(A=@n31.Bias, B=Multiply(A=Multiply(A=@n38, B=@n40), B=Add(A=@n41, B=@n31.Param3) {constB=0.5}))
  n43 = ObjectPositionWS()
  n44 = WorldPosition()
  Add(
   A=Multiply(A=VertexColor().R, B=RotateAboutAxis(NormalizedRotationAxis=@n28, RotationAngle=@n34, PivotPoint=@n35, Position=@n36.XYZ))
   B=Multiply(A=VertexColor().G, B=RotateAboutAxis(NormalizedRotationAxis=@n37, RotationAngle=@n42, PivotPoint=@n43, Position=@n44.XYZ))
  )
  Constant()
  n45 = Constant() {r=1}
  If(A=@n3.Speed, B=Constant(), "A > B"=@n3.Speed, "A == B"=@n45, "A < B"=@n45)
  n46 = TextureSampleParameter2D() {
   parameterName=ColorTexture
   texture=/BRCosmetics/Athena/Effects/Particles/Skydiving/NS_Contrail_LazarusLens/Textures/T_LazarusLense_Contrail_D
  }
  LinearInterpolate(
   A=ScalarParameter() {defaultValue=1, parameterName=RoughnessHigh}
   B=ScalarParameter() {defaultValue=0.5, parameterName=RoughnessLOW}
   Alpha=@n46.R
  )
  n47 = ParticleColor()
  n48 = LinearInterpolate(
   A=VectorParameter() {parameterName=Color2, defaultValue=<rgba 0.3 0.3 0.3 1>}.RGB
   B=VectorParameter() {parameterName=Color1, defaultValue=<rgba 0.515 0.515 0.515 1>}.RGB
   Alpha=ComponentMask(None=TextureCoordinate() {coordinateIndex=1}) {r=true}
  )
  n49 = Multiply(A=@n46.RGB, B=ParticleColor().RGB)
  n50 = Multiply(A=@n48, B=@n49)
  MaterialFunctionCall(Emissive=Add(A=ScalarParameter() {parameterName=Emissive}, B=Multiply(A=@n47.RGB, B=@n50))) {materialFunction=&MF_FX_EmissiveAdjustments}
  Multiply(A=@n48, B=@n49)
  n51 = Add(B=ScalarParameter() {parameterName=Macro Time Bias})
  n52 = ScalarParameter() {defaultValue=1, parameterName=Speed}
  n53 = ComponentMask(None=Transform(None=WorldPosition().XYZ) {transformSourceType=TRANSFORMSOURCE_World, transformType=TRANSFORM_Local}) {r=true}
  n54 = LinearInterpolate(
   A=ScalarParameter() {defaultValue=128, parameterName=Macro Wave Frequency}
   B=ScalarParameter() {defaultValue=128, parameterName=Macro Wave Frequency Wings}
   Alpha=Multiply(A=VertexColor().A, B=ScalarParameter() {parameterName=UseMacroWingFreq?})
  )
  n55 = Sine(None=Add(A=Multiply(A=@n51, B=@n52), B=Divide(A=@n53, B=@n54)))
  n56 = ScalarParameter() {defaultValue=1, parameterName=Macro Wave scale}
  Multiply(
   A=Transform(None=Constant3Vector() {constant=<rgba 0 0 1 0>}) {transformSourceType=TRANSFORMSOURCE_Local}
   B=Multiply(A=Multiply(A=@n55, B=@n56), B=VertexColor().B)
  )
  n57 = DynamicParameter() {defaultValue=<rgba 1 1 0 0>, paramNames [Bias,Speed,Param3,Param4]}
  n58 = Time()
  n59 = Add(A=Sine(None=Multiply(A=@n57.Speed, B=@n58)) {period=0.75}, B=@n57.Param3)
  n60 = Transform(None=Constant3Vector() {constant=<rgba 0 0 1 0>}) {transformSourceType=TRANSFORMSOURCE_Local}
  Multiply(A=VertexColor(), B=Multiply(A=Multiply(A=@n60, B=@n59), B=@n57.Bias) {constB=5})
  TextureSampleParameter2D() {
   parameterName=NormalTexture
   samplerType=SAMPLERTYPE_Normal
   texture=/BRCosmetics/Athena/Effects/Particles/Skydiving/NS_Contrail_LazarusLens/Textures/T_LazarusLense_Contrail_N
  }
  VectorParameter() {parameterName=Color, defaultValue=<rgba 1 1 1 1>}
 ]
}
