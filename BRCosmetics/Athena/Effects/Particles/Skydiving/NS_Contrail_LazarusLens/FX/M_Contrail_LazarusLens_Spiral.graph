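# Translucent, unlit, two-sided mesh-particle material for the LazarusLens skydiving contrail, drawing an animated circular spiral effect.
# Time-driven sine rings in radial UVs combine with circular masks, eroded by panning polar cloud noise sampled with original-UV derivatives.
# The mask drives opacity and a curve-atlas color ramp, with time-of-day emissive scaling and optional particle color.
# Radius- and angle-based sine waves ripple the mesh along local Z through world position offset.
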Material /BRCosmetics/Athena/Effects/Particles/Skydiving/NS_Contrail_LazarusLens/FX/M_Contrail_LazarusLens_Spiral
graph {
 properties {
  bUsedWithNiagaraMeshParticles=true
  blendMode=BLEND_Translucent
  pixelDepthOffsetMode=PDOM_Legacy
  shadingModel=MSM_Unlit
  twoSided=true
  usedShadingModels=Unlit
 }
 nodes [
  DynamicParameter() {parameterIndex=3, defaultValue=<rgba 1 0.2 -0.02 1>, paramNames [Param1,Speed,WPO,Param4]}
  n1 = SphereMask(A=TextureCoordinate(), B=Constant2Vector() {g=0.5, r=0.5}) {attenuationRadius=0.49, hardnessPercent=85}
  n2 = SphereMask(A=TextureCoordinate(), B=Constant2Vector() {g=0.5, r=0.5}) {attenuationRadius=0.2, hardnessPercent=30}
  n3 = Time()
  n4 = Constant() {r=-0.02}
  n5 = MaterialFunctionCall(Vector or UVs=ConstantBiasScale(None=TextureCoordinate()) {bias=-0.5, scale=2}) {materialFunction=&VectorToRadialValue}
  n6 = ComponentMask(None=@n5.Radial Coordinates) {g=true}
  n7 = Power(Base=Multiply(B=@n6) {constA=0.5}, Exp=ScalarParameter() {defaultValue=1.5, parameterName=RingAmount}) {constExponent=1.5}
  n8 = Sine(None=Add(A=Multiply(A=@n3, B=@n4), B=@n7)) {period=0.1}
  n9 = SmoothStep(Value=@n8) {constMin=0.5}
  n10 = Multiply(A=SmoothStep(Value=@n8) {constMin=0.6}, B=ScalarParameter() {defaultValue=1, parameterName=HardEdge}) {constB=0.9}
  n11 = Add(
   A=LinearInterpolate(B=OneMinus(None=@n1), Alpha=ScalarParameter() {defaultValue=0.15, parameterName=Border})
   B=Multiply(A=OneMinus(None=@n2), B=Subtract(A=@n9, B=@n10))
  )
  n12 = Constant()
  n13 = DynamicParameter() {parameterIndex=3, defaultValue=<rgba 1 0.2 -0.02 1>, paramNames [Param1,Speed,WPO,Param4]}
  n14 = TextureCoordinate()
  n15 = TextureSampleParameter2D!(
   UVs=Panner(
    Coordinate=MaterialFunctionCall() {materialFunction=&VectorToRadialValue}.Radial Coordinates
    Speed=MaterialFunctionCall(X=@n12, Y=@n13.Speed) {materialFunction=&MakeFloat2}.Result
   ) {speedY=-0.1}
   "DDX(UVs)"=DDX(Value=@n14)
   "DDY(UVs)"=DDY(Value=@n14)
   Apply View MipBias
  ) {
   group=Wispy Flame
   mipValueMode=TMVM_Derivative
   parameterName=Flame Texture
   samplerType=SAMPLERTYPE_LinearGrayscale
   texture=/Game/Effects/Fort_Effects/Emotes/UltraLight/Texture/T_Cloud_Greyscale
  }
  n16 = Multiply(A=SmoothStep(Value=@n15.R) {constMax=0.6, constMin=0.4}, B=@n1)
  n17 = SmoothStep(Value=Saturate(None=Subtract(A=@n11, B=@n16))) {constMax=0.3}
  n18 = SmoothStep(
   Min=ScalarParameter() {group=Color, parameterName=Color Smooth Step Min, sliderMax=1, sliderMin=-1}
   Max=ScalarParameter() {defaultValue=2, group=Color, parameterName=Color Smooth Step Max, sliderMax=5}
   Value=Power(Base=Saturate(None=@n17), Exp=ScalarParameter() {defaultValue=0.01, group=Color, parameterName=Color Power, sliderMax=5}) {constExponent=1}
  ) {constMax=2}
  n19 = CurveAtlasRowParameter(CurveTime=@n18) {
   atlas=/Game/Athena/LinearColorCurves/CurveAtlas_Chainmail
   curve=/Game/Athena/LinearColorCurves/Curve_TealBlack
   defaultValue=4
   parameterName=Curve_Color
  }
  n20 = ScalarParameter() {defaultValue=1, parameterName=Emissive, sliderMax=2}
  n21 = Multiply(A=@n19, B=Add(A=Multiply(A=@n18, B=@n20)) {constB=0.5})
  n22 = VectorParameter() {parameterName=TOD color, defaultValue=<rgba 1 1 1 0.5>}
  n23 = MaterialFunctionCall(Value_Morning=@n22.R, Value_Day=@n22.R, Value_Evening=@n22.R, Value_Night=@n22.R) {materialFunction=&Scalar_Time_of_Day}
  n24 = ParticleColor()
  MaterialFunctionCall(Emissive=LinearInterpolate(A=Multiply(A=@n21, B=@n23.Result), B=@n24.RGB, Alpha=ScalarParameter() {parameterName=ParticleColor})) {materialFunction=&MF_FX_EmissiveAdjustments}
  Multiply(A=@n24.A, B=Add(A=@n17, B=ScalarParameter() {defaultValue=0.05, parameterName=OverallOpacity}))
  n25 = Constant3Vector() {constant=<rgba 0 0 1 0>}
  n26 = Multiply(A=Time(), B=Constant() {r=-0.1})
  n27 = Add(A=@n26, B=@n7)
  n28 = ComponentMask(None=@n5.Radial Coordinates) {r=true}
  n29 = Add(A=Add(A=@n6, B=Multiply(A=@n28) {constB=0.5}), B=@n26)
  n30 = Multiply(
   A=Multiply(A=Sine(None=@n27) {period=0.3}, B=Sine(None=@n29) {period=0.5})
   B=SphereMask(A=TextureCoordinate(), B=Constant2Vector() {g=0.5, r=0.5}) {attenuationRadius=0.45, hardnessPercent=80}
  )
  Multiply(
   A=ScalarParameter() {defaultValue=3, parameterName=WPO Intensity, sliderMax=100}
   B=Transform(None=Multiply(A=@n25, B=@n30)) {transformSourceType=TRANSFORMSOURCE_Local}
  )
  TextureCoordinate() {uTiling=2, vTiling=2}
  TextureCoordinate()
 ]
}
