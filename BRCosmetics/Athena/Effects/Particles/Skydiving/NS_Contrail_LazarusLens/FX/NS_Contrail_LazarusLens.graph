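# Niagara skydiving contrail for the LazarusLens cosmetic, mixing a clock mesh, bats, birds, smoke and cyan glows.
# System-level diving attributes and a timeline value drive spawn rates, offsets, scale and brightness across emitters, with front-end preview and time-of-day adjustments.
# Bats and birds swirl with vortex forces and glow by height proximity to the system origin.
# Smoke tails sample skeletal-mesh sockets or bones and align to dive velocity, alongside GPU sparks.
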NiagaraSystem /BRCosmetics/Athena/Effects/Particles/Skydiving/NS_Contrail_LazarusLens/FX/NS_Contrail_LazarusLens
dependencies {
 usedDataInterfaces [
  &NiagaraDataInterfaceCurve
  &NiagaraDataInterfaceMeshRendererInfo
  &NiagaraDataInterfaceVectorCurve
  &NiagaraDataInterfaceColorCurve
  &NiagaraDataInterfaceVector2DCurve
  &NiagaraDataInterfaceArrayFloat
 ]
 usedDynamicInputs [&TransformVector]
 usedModules [
  &SystemState
  &self:SystemSpawnScript.SystemScriptSource.SystemScriptGraph.NiagaraNodeAssignment_0.SetVariables_#1
  /Game/Effects/Niagara/Modules/Skydiving/FN_DivingAttributes/v4/FN_DivingAttributes_v4
  &self:SystemSpawnScript.SystemScriptSource.SystemScriptGraph.NiagaraNodeAssignment_1.SetVariables_#2
  &self:SystemSpawnScript.SystemScriptSource.SystemScriptGraph.NiagaraNodeAssignment_2.SetVariables_#3
  &self:SystemSpawnScript.SystemScriptSource.SystemScriptGraph.NiagaraNodeAssignment_3.SetVariables_#4
  &self:SystemSpawnScript.SystemScriptSource.SystemScriptGraph.NiagaraNodeAssignment_4.SetVariables_#5
  &Timeline
 ]
}
emitters {
 Bats {
  emitterUpdateScript [&EmitterState()]
  particleSpawnScript [&InitializeParticle(
    Color=<rgba 0.0123 0.0605 0.0729 1>
    Lifetime Max=1.5
    Lifetime Min=1.2
    Lifetime Mode=::NewEnumerator1
    Mesh Renderer Array Visibility Mode=::NewEnumerator3
    Mesh Renderer Info=DataInterface() {propertyValues {MeshRenderer=&self:Bats_7.NiagaraMeshRendererProperties_0}}
    Mesh Scale Mode=::NewEnumerator3
    Mesh Uniform Scale=&RandomRangeFloat(Minimum=0.35, Maximum=0.75)
    Position Mode=::NewEnumerator0
    Sprite Size Mode=::NewEnumerator3
    Uniform Sprite Size=48.6394
    Write Material Random=true
   )]
  particleUpdateScript [
   /SetVariables(
    Particles.Initial.Color:LinearColor=<rgba 0.0049 0.0258 0.026 1>
    Particles.Initial.Scale:Vector3f=&VectorFromFloat(Value=&RandomRangeFloat(Minimum=0.6, Maximum=1))
   ) {isSetParameters=true}
   &ParticleState()
   &Color() {enabled=false}
   &ScaleVelocity(Velocity Scale=&VectorFromCurve(VectorCurve=DataInterface() {propertyValues {
       XCurve {keys [
         {arriveTangent=4.1785, interpMode=RCIM_Cubic, leaveTangent=4.1785, tangentMode=RCTM_User, time=0, value=0}
         {arriveTangent=1.1179, interpMode=RCIM_Cubic, leaveTangent=1.1179, tangentMode=RCTM_User, time=0.2042, value=0.6228}
         {arriveTangent=0.0224, interpMode=RCIM_Cubic, leaveTangent=0.0224, tangentMode=RCTM_User, time=1, value=1}
        ]}
       YCurve {keys [
         {arriveTangent=4.1785, interpMode=RCIM_Cubic, leaveTangent=4.1785, tangentMode=RCTM_User, time=0, value=0}
         {arriveTangent=1.1179, interpMode=RCIM_Cubic, leaveTangent=1.1179, tangentMode=RCTM_User, time=0.2042, value=0.6228}
         {arriveTangent=0.0224, interpMode=RCIM_Cubic, leaveTangent=0.0224, tangentMode=RCTM_User, time=1, value=1}
        ]}
       ZCurve {keys [
         {arriveTangent=4.1785, interpMode=RCIM_Cubic, leaveTangent=4.1785, tangentMode=RCTM_User, time=0, value=0}
         {arriveTangent=1.1179, interpMode=RCIM_Cubic, leaveTangent=1.1179, tangentMode=RCTM_User, time=0.2042, value=0.6228}
         {arriveTangent=0.0224, interpMode=RCIM_Cubic, leaveTangent=0.0224, tangentMode=RCTM_User, time=1, value=1}
        ]}
      }})) {enabled=false}
   &VortexForce(
    Origin Pull Amount:NiagaraFloat=&RandomRangeFloat(Minimum=200, Maximum=300)
    UseOriginOffset=true
    Vortex Force Amount=500
    Vortex Origin Offset:Vector3f=<0 0 10>
   )
   &Drag(Drag=2, Use Linear Drag=true, Use Rotational Drag=true)
   &PointAttractionForce(
    Attraction Radius=50
    Attraction Strength=-150
    Attractor Position:NiagaraPosition=&AddVectorToPosition(Position=$Engine.Emitter.SimulationPosition, Vector=<0 0 70>)
    Kill Radius=5
    Kill Within Radius=false
    Module.Use Falloff=true
    UseAttractorOffset=false
   )
   &SolveForcesAndVelocity(Clamp Velocity=false, Limit Acceleration=false)
   &ScaleMeshSize(Scale Factor=&VectorFromFloat(Value=&FloatFromCurve(
      FloatCurve=DataInterface() {propertyValues {Curve {keys [
          {arriveTangent=18.4687, interpMode=RCIM_Cubic, leaveTangent=18.4687, tangentMode=RCTM_User, time=0, value=0}
          <at 0.1 1 RCIM_Cubic>
          {arriveTangent=-4.2304, interpMode=RCIM_Cubic, leaveTangent=-4.2304, tangentMode=RCTM_User, time=1, value=0}
         ]}}}
      CurveIndex=$StackContext.NormalizedAge
     )))
   &ScaleColor(
    Scale Alpha=&FloatFromCurve(FloatCurve=DataInterface() {propertyValues {Curve {keys [
         <at 0 0 RCIM_Cubic>
         {arriveTangent=1.25, leaveTangent=1.25, time=0.2, value=1}
         {interpMode=RCIM_Cubic, tangentMode=RCTM_User, time=0.8, value=1}
         {interpMode=RCIM_Cubic, leaveTangent=-16.5933, time=1, value=0}
        ]}}})
    Scale RGB=&Multiply_VectorByFloat(Vector=<1 1 1>, Float=$System.TOD)
    ScaleA=true
    ScaleRGB=true
   )
   &DynamicMaterialParameters(
    Index 0 Param 1=&RandomRangeFloat(Minimum=0, Maximum=3)
    Index 0 Param 2=&RandomRangeFloat(Minimum=1, Maximum=1.25)
    Module.Index0Param0Write=true
    Module.Index0Param1Write=true
    Module.Index0Param2Write=true
    Module.Index0Param3Write=true
   )
   /SetVariables(Particles.Scale=&VectorFromFloat(Value=&FloatFromCurve(FloatCurve=DataInterface() {propertyValues {Curve {keys [
          {arriveTangent=9.0725, interpMode=RCIM_Cubic, leaveTangent=9.0725, tangentMode=RCTM_User, time=0, value=0}
          <at 0.1408 0.9847 RCIM_Cubic>
          {arriveTangent=-4.2304, interpMode=RCIM_Cubic, leaveTangent=-4.2304, tangentMode=RCTM_User, time=1, value=0}
         ]}}}))) {enabled=false, isSetParameters=true}
   /SetVariables(Particles.DistanceFromCenter:NiagaraFloat=&Abs_Float(Float:NiagaraFloat=&Subtract_Float(
      A=&MakeFloatFromVector(
       Vector=&Add_Vector(
        A=/Niagara/DynamicInputs/Helpers/SimulationPosition()
        B=&Lerp_Vector(A=<0 0 30>, B=<0 0 0>, Alpha=$System.Timeline.NormalizedTime)
       )
       Channel=::NewEnumerator2
      )
      B=&MakeFloatFromVector(Vector=Linked() {linkedVariable {name=Particles.Position, type=NiagaraPosition}}, Channel=::NewEnumerator2)
     ))) {isSetParameters=true}
   /SetVariables(Particles.GlowProx:NiagaraFloat=&OneMinusFloat(Float=&NormalizeFloat(Value=$Particles.DistanceFromCenter, Min=0, Max=100))) {isSetParameters=true}
   /SetVariables(Particles.Dark_Glow=&MakeLinearColorFromVectorAndFloat(
     "Vector (RGB)"=&MakeVectorFromLinearColorRGB(Color=&ColorFromCurve(
       ColorCurve=DataInterface() {propertyValues {
         AlphaCurve {keys [<at 0 1 RCIM_Cubic>]}
         BlueCurve {keys [<at 0.3375 0.0312 RCIM_Cubic>,<at 1 0.5029 RCIM_Cubic>]}
         GreenCurve {keys [<at 0.3375 0.0309 RCIM_Cubic>,<at 1 0.3467 RCIM_Cubic>]}
         RedCurve {keys [<at 0.3375 0.0293 RCIM_Cubic>,<at 1 0.0356 RCIM_Cubic>]}
        }}
       CurveIndex=$Particles.GlowProx
      ))
     "Float (Alpha)"=&Multiply_Float(
      A=&FloatFromCurve(FloatCurve=DataInterface() {propertyValues {Curve {keys [
           {arriveTangent=9.0725, interpMode=RCIM_Cubic, leaveTangent=9.0725, tangentMode=RCTM_User, time=0, value=0}
           {arriveTangent=-2.5249, interpMode=RCIM_Cubic, leaveTangent=-2.5249, tangentMode=RCTM_User, time=0.1408, value=0.9847}
           {arriveTangent=-0.0405, interpMode=RCIM_Cubic, leaveTangent=-0.0405, tangentMode=RCTM_User, time=1, value=0}
          ]}}})
      B=&Lerp_Float(A=1, B=1, Alpha=$Particles.GlowProx)
     )
    )) {isSetParameters=true}
   /SetVariables(Particles.Dark_Glow_Size=&Multiply_Vector2(
     A=&Vector2DFromFloat(Value=80)
     B=&Vector2DFromFloat(Value=&FloatFromCurve(
       FloatCurve=DataInterface() {propertyValues {Curve {keys [<at 0 1.5 RCIM_Cubic>,<at 1 0.5>]}}}
       Scale Curve=&FN_FrontEndFloatModifier(Gameplay=1, Front End=0.6)
      ))
    )) {isSetParameters=true}
  ]
  rendererClasses [&NiagaraMeshRendererProperties, &NiagaraSpriteRendererProperties]
  renderers [
   {rendererClass=&NiagaraMeshRendererProperties, rendererIndex=0}
   {rendererClass=&NiagaraSpriteRendererProperties, rendererIndex=1}
   {rendererClass=&NiagaraSpriteRendererProperties, rendererIndex=2}
  ]
 }
 Birds {
  emitterUpdateScript [
   &EmitterState(
    Inactive Response=::NewEnumerator1
    Life Cycle Mode=::NewEnumerator1
    Loop Duration=&RandomRangeFloat(Minimum=0.75, Maximum=1.5)
    UseLoopDelay=false
   )
   &SpawnRate(
    Spawn Probability=0.6
    SpawnRate=&MakeCustomFloatFromBool(
     Bool=$User.bIsFrontEndPreview
     True Float=2
     False Float=&MakeFloatFromVector(
      Vector=&VectorFromFloat(Value=&Lerp_Float(A=2, B=5, Alpha=$System.Timeline.NormalizedTime))
      Channel=::NewEnumerator0
     )
    )
    Use Spawn Probability=true
   )
  ]
  particleSpawnScript [
   &PartitionParticles(Partition Mode:ENiagara_PartitionMode=::NewEnumerator4) {enabled=false}
   &InitializeParticle(
    Lifetime=5
    Mesh Scale Mode=::NewEnumerator3
    Mesh Uniform Scale=&RandomRangeFloat(Minimum=0.2, Maximum=0.3)
    UsePositionOffset=false
    Write Material Random=true
   )
   &SkeletalMeshLocation(
    Bone / Socket Index=$Particles.PartitionParticles.Partition
    MaskSpawnGroup=false
    Mesh Sampling Type=::NewEnumerator5
    Orientation Sampling=::NewEnumerator1
    Skeletal Mesh=$User.SkeletalMesh
    Socket Sampling Mode=::NewEnumerator5
   ) {enabled=false}
   &SystemLocation(MaskSpawnGroup=false, Offset=<0 0 -200>)
   &CylinderLocation(
    Cylinder Height=10
    Cylinder Height Midpoint=1
    Cylinder Radius=&Lerp_Float(A=65, B=75, Alpha=$System.Timeline.NormalizedTime)
    MaskSpawnGroup=false
    Offset Mode:ENiagara_OffsetMode=::NewEnumerator0
    Override Seed=false
    UseOwnerScale=true
   )
   &InitialMeshOrientation(
    Mesh Axis To Orient=<0 1 0>
    Mesh Orientation Mode=::NewEnumerator2
    Orientation Vector=<-1 0 0>
    Rotation=&RandomRangeVector(Minimum=<-0.05 0.05 0>, Maximum=<0.05 0.1 0.5>)
    Rotation Coordinate Space=::NewEnumerator2
    Use Rotation Vector=true
   )
   &AddVelocity(Velocity=&RandomRangeVector(Minimum=<0 0 225>, Maximum=<0 0 300>))
   /SetVariables() {isSetParameters=true}
   &ApplyInitialForces() {enabled=false}
  ]
  particleUpdateScript [
   &InheritSourceMovement(
    Applied Position Scale=&Lerp_Vector(
     A=<0.98 0.98 0.99>
     B=&RandomRangeVector(Minimum=<0.98 0.98 0.98>, Maximum=<0.985 0.985 0.985>)
     Alpha=$System.Timeline.NormalizedTime
    )
    ApplyForce=false
    ApplyPosition=true
    ApplyPositionFalloff=false
   ) {enabled=false}
   &ParticleState()
   &InheritVelocity(
    Inherited Velocity Amount Scale=&Lerp_Vector(A=<0.5 0.5 0>, B=<1 1 1>, Alpha=$System.Timeline.NormalizedTime)
    Inherited Velocity Speed Limit=1
    Speed Limit=true
    UseSpeedThreshold=false
   ) {enabled=false}
   &WindForce(
    Offset By Time:Vector3f=<0 0 -10>
    Scale=0.6089
    Wind Speed=&RandomRangeVector(Minimum=<-20 -20 50>, Maximum=<20 20 100>)
    Wind Speed Scale=&Lerp_Float(A=1, B=0.2, Alpha=$System.Timeline.NormalizedTime)
   )
   &VortexForce(
    UseOriginOffset=false
    Vortex Force Amount=&Lerp_Float(A=100, B=80, Alpha=$System.Timeline.NormalizedTime)
    Vortex Origin:NiagaraPosition=$Particles.SkeletalMeshLocation.SampledPosition
   )
   &Drag(Drag=&RandomRangeFloat(Minimum=0.5, Maximum=1), Use Linear Drag=true, Use Rotational Drag=true)
   &SolveForcesAndVelocity(Clamp Velocity=false, Limit Acceleration=false)
   &ScaleMeshSize(Scale Factor=&VectorFromFloat(Value=&FloatFromCurve(
      FloatCurve=DataInterface() {propertyValues {Curve {keys [
          {arriveTangent=7.9421, interpMode=RCIM_Cubic, leaveTangent=7.9421, tangentMode=RCTM_User, time=0, value=0}
          {interpMode=RCIM_Cubic, tangentMode=RCTM_User, time=0.5, value=1}
          {arriveTangent=-6.2155, interpMode=RCIM_Cubic, leaveTangent=-6.2155, tangentMode=RCTM_User, time=1, value=0}
         ]}}}
      Scale Curve=2
     )))
   &ScaleColor(
    Scale Alpha=0.9
    Scale RGB=&VectorFromFloat(Value=&Lerp_Float(A=1, B=3, Alpha=$System.Timeline.NormalizedTime))
    ScaleA=true
    ScaleRGB=true
   )
   &DynamicMaterialParameters(
    Index 0 Param 1=1.5
    Index 0 Param 2=&RandomRangeFloat(Minimum=4, Maximum=5)
    Module.Index0Param0Write=true
    Module.Index0Param1Write=true
    Module.Index0Param2Write=true
    Module.Index0Param3Write=true
   )
   &KillParticlesInVolume(
    Kill Shape=::NewEnumerator2
    Origin Offset:Vector3f=&Lerp_Vector(A=<0 0 25>, B=<0 0 0>, Alpha=$System.Timeline.NormalizedTime)
    Plane Normal:Vector3f=<0 0 -1>
   )
   &GenerateDeathEvent(Use Event Probability=false)
   /SetVariables(Particles.SpawnEvent:NiagaraBool=&SetBoolByFloatComparison(
     A=&MakeFloatFromVector(Vector=Linked() {linkedVariable {name=Particles.Position, type=NiagaraPosition}}, Channel=::NewEnumerator2)
     B=&MakeFloatFromVector(Vector=&Add_Vector(A=&SimulationPosition(), B=<0 0 -20>), Channel=::NewEnumerator2)
    )) {enabled=false, isSetParameters=true}
   &GenerateLocationEvent(
    Event Generation Enabled:NiagaraBool=$Particles.SpawnEvent
    Event Type:ENiagara_LocEventType=::NewEnumerator2
    Use Event Probability=false
    UseEventDelay=false
   ) {enabled=false}
   /SetVariables(Particles.DistanceFromCenter:NiagaraFloat=&Abs_Float(Float:NiagaraFloat=&Subtract_Float(
      A=&MakeFloatFromVector(
       Vector=&Add_Vector(
        A=&SimulationPosition()
        B=&Lerp_Vector(A=<0 0 30>, B=<0 0 0>, Alpha=$System.Timeline.NormalizedTime)
       )
       Channel=::NewEnumerator2
      )
      B=&MakeFloatFromVector(Vector=Linked() {linkedVariable {name=Particles.Position, type=NiagaraPosition}}, Channel=::NewEnumerator2)
     ))) {isSetParameters=true}
   /SetVariables(Particles.GlowProx:NiagaraFloat=&OneMinusFloat(Float=&NormalizeFloat(Value=$Particles.DistanceFromCenter, Min=0, Max=30))) {isSetParameters=true}
   /SetVariables(Particles.Dark_Glow=&MakeLinearColorFromVectorAndFloat(
     "Vector (RGB)"=&MakeVectorFromLinearColorRGB(Color=<rgba 0.0354 0.3478 0.5 1>)
     "Float (Alpha)"=&Multiply_Float(
      A=&FloatFromCurve(
       FloatCurve=DataInterface() {propertyValues {Curve {keys [
           {arriveTangent=9.0725, interpMode=RCIM_Cubic, leaveTangent=9.0725, tangentMode=RCTM_User, time=0, value=0}
           {arriveTangent=-2.5249, interpMode=RCIM_Cubic, leaveTangent=-2.5249, tangentMode=RCTM_User, time=0.1408, value=0.9847}
           {arriveTangent=-0.0405, interpMode=RCIM_Cubic, leaveTangent=-0.0405, tangentMode=RCTM_User, time=1, value=0}
          ]}}}
       Scale Curve=0.5
      )
      B=&Lerp_Float(A=0, B=1, Alpha=$Particles.GlowProx)
     )
    )) {isSetParameters=true}
   /SetVariables(Particles.Dark_Glow_Size=&Multiply_Vector2(
     A=&Vector2DFromFloat(Value=80)
     B=&Vector2DFromFloat(Value=&FloatFromCurve(
       FloatCurve=DataInterface() {propertyValues {Curve {keys [<at 0 1.5 RCIM_Cubic>,<at 1 0.5>]}}}
       Scale Curve=&FN_FrontEndFloatModifier(Gameplay=1, Front End=0.6)
      ))
    )) {isSetParameters=true}
  ]
  rendererClasses [&NiagaraMeshRendererProperties, &NiagaraSpriteRendererProperties]
  renderers [
   {rendererClass=&NiagaraMeshRendererProperties, rendererIndex=0}
   {rendererClass=&NiagaraSpriteRendererProperties, rendererIndex=1}
  ]
 }
 Burst_Air_Meshes {
  emitterUpdateScript [
   &EmitterState()
   &SpawnBurst_Instantaneous(
    Spawn Count=&RandomRangeInt(Minimum=8, Maximum=12)
    Spawn Time=1.15
    Use Spawn Probability=false
    UseLoopCountLimit=false
   ) {enabled=false}
  ]
  particleSpawnScript [
   &InitializeParticle(
    Color=&UniformRangedLinearColor(Minimum=<rgba 0.0497 0.0701 0.085 1>, Maximum=<rgba 0.0318 0.0517 0.0625 1.25>)
    Lifetime Max=0.6
    Lifetime Min=0.5
    Lifetime Mode=::NewEnumerator1
    Mass Mode=::NewEnumerator1
    Mesh Scale=&Multiply_VectorByFloat(Vector=<1 1 3>, Float=0.11)
    Mesh Scale Mode=::NewEnumerator4
    Sprite Size=<1 1>
    Sprite Size Mode=::NewEnumerator4
    UsePositionOffset=false
    Write Material Random=true
   )
   &SystemLocation(
    MaskSpawnGroup=false
    Offset=&VectorFromCurve(
     VectorCurve=DataInterface() {propertyValues {XCurve {keys [<at 0 1>,<at 1 0>]}, YCurve {keys [<at 0 2.5>,<at 1 2.5>]}, ZCurve {keys [<at 0 30>,<at 1 90>]}}}
     CurveIndex=$Particles.MaterialRandom
    )
    Offset Coordinate Space=::Simulation
   ) {enabled=false}
  ]
  particleUpdateScript [
   /SetVariables(
    Particles.Initial.Color:LinearColor=&UniformRangedLinearColor(Minimum=<rgba 0.0497 0.0701 0.085 1>, Maximum=<rgba 0.0318 0.0517 0.0625 1.5>)
    Particles.Initial.Scale:Vector3f=&Multiply_VectorByFloat(Vector=<1.5 1.5 6>, Float=0.11)
   ) {isSetParameters=true}
   &ParticleState()
   &UpdateMeshOrientation(Rotation Rate=&UniformRangedFloat(Minimum=1.5, Maximum=2), Rotation Vector=<0 0 1>)
   &ScaleMeshSize(Scale Factor=&VectorFromCurve(
     VectorCurve=DataInterface() {propertyValues {
       XCurve {keys [{arriveTangent=1.2832, interpMode=RCIM_Cubic, leaveTangent=1.2832, tangentMode=RCTM_User, time=0, value=0.75}, <at 1 2.5 RCIM_Cubic>]}
       YCurve {keys [{arriveTangent=1.2832, interpMode=RCIM_Cubic, leaveTangent=1.2832, tangentMode=RCTM_User, time=0, value=0.75}, <at 1 2.5 RCIM_Cubic>]}
       ZCurve {keys [<at 0 1>,<at 1 0>]}
      }}
     Scale Curve=&RandomRangeVector(Minimum=<1 1 1>, Maximum=<1 1 1>)
    ))
   &ScaleColor(
    Scale Alpha=&FloatFromCurve(FloatCurve=DataInterface() {propertyValues {Curve {keys [
         {arriveTangent=7.0264, interpMode=RCIM_Cubic, leaveTangent=7.0264, tangentMode=RCTM_User, time=0, value=0}
         {arriveTangent=0.2184, interpMode=RCIM_Cubic, leaveTangent=0.2184, tangentMode=RCTM_User, time=0.2, value=1}
         {arriveTangent=-0.8655, interpMode=RCIM_Cubic, leaveTangent=-0.8655, tangentMode=RCTM_User, time=1, value=0}
        ]}}})
    ScaleA=true
    ScaleRGB=false
   )
  ]
  rendererClasses [&NiagaraMeshRendererProperties]
  renderers [{rendererClass=&NiagaraMeshRendererProperties, rendererIndex=0}]
 }
 Clock {
  emitterUpdateScript [
   &EmitterState(Life Cycle Mode=::NewEnumerator1, Loop Behavior=::NewEnumerator1, Loop Duration Mode=::NewEnumerator1, UseLoopDelay=false)
   &SpawnBurst_Instantaneous(Spawn Time=0.35, Use Spawn Probability=false, UseLoopCountLimit=false)
  ]
  particleSpawnScript [
   &InitializeParticle(
    Color=<rgba 0.1782 1.3504 1.5 1>
    Mesh Scale=<1.7 1.7 1.7>
    Mesh Scale Mode=::NewEnumerator4
    UsePositionOffset=false
    Write Material Random=true
   )
   &InitialMeshOrientation(Mesh Orientation Mode=::NewEnumerator2, Use Rotation Vector=true)
  ]
  particleUpdateScript [
   &ParticleState(Kill Particles When Lifetime Has Elapsed=false, Let Infinitely Lived Particles Die When Emitter Deactivates=true)
   &SystemLocation(MaskSpawnGroup=false, Offset=&Lerp_Vector(A=<0 0 40>, B=<0 0 10>, Alpha=$System.Timeline.NormalizedTime))
   &ScaleMeshSize(Scale Factor=&MakeCustomVectorFromBool(
     Bool=$User.bIsFrontEndPreview
     True Vector=<1.05 1.05 1.05>
     False Vector=&Lerp_Vector(A=<1 1 1>, B=<1.25 1.25 1.25>, Alpha=$System.Timeline.NormalizedTime)
    )) {module=ScaleMeshSize001}
   &UpdateMeshOrientation(Rotation Vector=<0 0 1>) {enabled=false}
   &ScaleMeshSize(Scale Factor=&VectorFromFloat(Value=&Lerp_Float(A=1, B=1.5, Alpha=$System.Dive))) {enabled=false}
   &ScaleColor(
    Scale Alpha=&Multiply_Float(
     A=&FloatFromCurve(
      FloatCurve=DataInterface() {propertyValues {Curve {keys [{arriveTangent=3.8917, interpMode=RCIM_Cubic, leaveTangent=3.8917, tangentMode=RCTM_User, time=0, value=0}, <at 0.5 1 RCIM_Cubic>]}}}
      CurveIndex=$Particles.Age
      Scale Curve=&FN_FrontEndFloatModifier(Gameplay=1, Front End=0.9)
     )
     B=&Lerp_Float(A=1, B=0.8, Alpha=$System.Timeline.NormalizedTime)
    )
    Scale RGB=&VectorFromFloat(Value=&Multiply_Float(
      A=&FloatFromCurve(
       FloatCurve=DataInterface() {propertyValues {Curve {keys [
           <at 0 0 RCIM_Cubic>
           {arriveTangent=8.0289, interpMode=RCIM_Cubic, leaveTangent=-5.7414, tangentMode=RCTM_Break, time=0.4, value=3}
           <at 0.9 1 RCIM_Cubic>
          ]}}}
       CurveIndex=$Particles.Age
       Scale Curve=&Multiply_Float(A=&FNTimeOfDayScalar(Night Scale=0.6, Morning Scale=1, Evening Scale=0.6), B=0.8)
      )
      B=&Lerp_Float(A=1, B=0, Alpha=$System.Timeline.NormalizedTime)
     ))
    ScaleA=true
    ScaleRGB=true
   )
   &DynamicMaterialParameters(
    Index 2 Param 1=&MakeCustomFloatFromBool(Bool=$System.IsDiving, True Float=0.7, False Float=0.2)
    Index 2 Param 2=&MakeCustomFloatFromBool(Bool=$System.IsDiving, True Float=0.7, False Float=0.2)
    Index 3 Param 1=&Lerp_Float(A=0, B=1, Alpha=$System.Timeline.NormalizedTime)
    Index 3 Param 2=&MakeCustomFloatFromBool(Bool=$System.IsDiving, True Float=-0.3, False Float=-0.1)
    Module.Param0WriteEnabled2=true
    Module.Param0WriteEnabled3=true
    Module.Param1WriteEnabled2=true
    Module.Param1WriteEnabled3=true
    Module.Param2WriteEnabled2=true
    Module.Param2WriteEnabled3=true
    Module.Param3WriteEnabled2=true
    Module.Param3WriteEnabled3=true
    Write Parameter Index 0=false
    Write Parameter Index 2=true
    Write Parameter Index 3=true
   )
   /SetVariables(Particles.CameraOffsetSecondary=-3, Particles.SecondaryColor:LinearColor=<rgba 0.0156 0.0156 0.0156 1>) {isSetParameters=true}
   &ScaleColor(
    Color Value To Scale=$Particles.SecondaryColor
    Scale Alpha=&Multiply_Float(
     A=&FloatFromCurve(
      FloatCurve=DataInterface() {propertyValues {Curve {keys [{arriveTangent=3.8917, interpMode=RCIM_Cubic, leaveTangent=3.8917, tangentMode=RCTM_User, time=0, value=0}, <at 0.5 1 RCIM_Cubic>]}}}
      CurveIndex=$Particles.Age
      Scale Curve=&FN_FrontEndFloatModifier(Gameplay=1, Front End=0.9)
     )
     B=&Lerp_Float(A=1, B=0.8, Alpha=$System.Timeline.NormalizedTime)
    )
    Scale RGB=&VectorFromFloat(Value=&FloatFromCurve(
      FloatCurve=DataInterface() {propertyValues {Curve {keys [
          <at 0 0 RCIM_Cubic>
          {arriveTangent=8.0289, interpMode=RCIM_Cubic, leaveTangent=-5.7414, tangentMode=RCTM_Break, time=0.4, value=3}
          <at 0.9 1 RCIM_Cubic>
         ]}}}
      CurveIndex=$Particles.Age
      Scale Curve=&Multiply_Float(A=&FNTimeOfDayScalar(Night Scale=0.6, Morning Scale=1, Evening Scale=0.6), B=0.8)
     ))
    ScaleA=true
    ScaleRGB=true
   ) {module=ScaleColor001}
   /SetVariables(Particles.MeshScaleBlack:Vector3f=&Multiply_VectorByFloat(Vector=$Particles.Scale, Float=0.97)) {isSetParameters=true}
  ]
  rendererClasses [&NiagaraMeshRendererProperties]
  renderers [
   {rendererClass=&NiagaraMeshRendererProperties, rendererIndex=0}
   {rendererClass=&NiagaraMeshRendererProperties, rendererIndex=1}
   {rendererClass=&NiagaraMeshRendererProperties, rendererIndex=2}
  ]
 }
 Color_wave {
  emitterUpdateScript [&EmitterState()]
  particleSpawnScript [&InitializeParticle(
    Lifetime=0.65
    Sprite Rotation Mode=::NewEnumerator2
    Sprite Size Mode=::NewEnumerator3
    Uniform Sprite Size=55
    UsePositionOffset=false
    Write Material Random=true
   )]
  particleUpdateScript [
   /SetVariables(
    Particles.Initial.Color:LinearColor=<rgba 0 3.1667 5 0.5>
    Particles.Initial.SpriteSize:Vector2f=&Vector2DFromFloat(Value=65)
   ) {isSetParameters=true}
   &ParticleState()
   &ScaleSpriteSize(Uniform Curve Sprite Scale=DataInterface() {propertyValues {Curve {keys [
        <at 0 0 RCIM_Cubic>
        {arriveTangent=1, interpMode=RCIM_Cubic, leaveTangent=1, time=0.3659, value=0.7136}
        {tangentMode=RCTM_User, time=1, value=1}
       ]}}})
   &DynamicMaterialParameters(
    Index 0 Param 1=1
    Index 0 Param 2=&FloatFromCurve(FloatCurve=DataInterface() {propertyValues {Curve {keys [
         {interpMode=RCIM_Cubic, tangentMode=RCTM_User, time=0, value=1}
         {arriveTangent=-1, interpMode=RCIM_Cubic, leaveTangent=-1, time=0.6, value=0.7}
         {interpMode=RCIM_Cubic, leaveTangent=-3.3753, time=1, value=0}
        ]}}})
    Index 0 Param 3=1
    Index 0 Param 4=1
    Module.Param0WriteEnabled=true
    Module.Param1WriteEnabled=true
    Module.Param2WriteEnabled=true
    Module.Param3WriteEnabled=true
   )
   &ScaleColor(
    Scale Alpha=&FloatFromCurve(FloatCurve=DataInterface() {propertyValues {Curve {keys [
         {interpMode=RCIM_Cubic, tangentMode=RCTM_User, time=0, value=1}
         {arriveTangent=-2.3888, interpMode=RCIM_Cubic, leaveTangent=-2.3888, tangentMode=RCTM_User, time=1, value=0}
        ]}}})
    Scale RGB=&MakeVectorFromLinearColorRGB(Color=&MultiplyLinearColor(
      Linear Color=&ColorFromCurve(ColorCurve=DataInterface() {propertyValues {
         AlphaCurve {keys []}
         BlueCurve {keys [<at 0.3264 3 RCIM_Cubic>, {interpMode=RCIM_Cubic, leaveTangent=-0.7175, time=0.6381, value=0.2118}]}
         GreenCurve {keys [<at 0.3264 3 RCIM_Cubic>, {interpMode=RCIM_Cubic, leaveTangent=-4.4353, time=0.6381, value=0.2552}]}
         RedCurve {keys [<at 0.3264 0.189 RCIM_Cubic>, {interpMode=RCIM_Cubic, leaveTangent=2.2176, time=0.6381, value=0.0143}]}
        }})
      Scale Factor=<1 1 1 1>
     ))
    ScaleA=true
    ScaleRGB=true
   )
   &SolveForcesAndVelocity(Clamp Velocity=false, Limit Acceleration=false)
   &CameraOffset(Camera Offset Amount=10)
  ]
  rendererClasses [&NiagaraSpriteRendererProperties]
  renderers [{rendererClass=&NiagaraSpriteRendererProperties}]
 }
 Flare {
  emitterUpdateScript [
   &EmitterState()
   &SpawnBurst_Instantaneous(Spawn Count=1, Use Spawn Probability=false) {enabled=false}
  ]
  particleSpawnScript [
   &InitializeParticle(
    Color=<rgba 0 0 0 1>
    Lifetime=0.25
    Sprite Rotation Mode=::NewEnumerator2
    Sprite Size Mode=::NewEnumerator3
    Uniform Sprite Size=200
    UsePositionOffset=false
    Write Material Random=true
   )
   /SetVariables(Particles.Dark_Glow_CameraOffset=&MakeCustomFloatFromBool(Bool=$User.bIsFrontEndPreview, True Float=1, False Float=50)) {enabled=false, isSetParameters=true, module=SetVariables_#6}
   &CameraOffset(Camera Offset Amount=&MakeCustomFloatFromBool(Bool=$User.bIsFrontEndPreview, True Float=1, False Float=30)) {enabled=false}
  ]
  particleUpdateScript [
   /SetVariables(Particles.Dark_Glow=&MakeLinearColorFromVectorAndFloat(
     "Vector (RGB)"=&MakeVectorFromLinearColorRGB(Color=<rgba 0.0114 0.1123 0.1615 1>)
     "Float (Alpha)"=&Multiply_Float(
      A=&Multiply_Float(A=&FNTimeOfDayScalar(Night Scale=0.7, Morning Scale=0.7, Evening Scale=1), B=0.5)
      B=&FloatFromCurve(FloatCurve=DataInterface() {propertyValues {Curve {keys [<at 0 0>,<at 0.2 1>,<at 0.7 0.6>,<at 1 0>]}}})
     )
    )) {isSetParameters=true, module=SetVariables_#7}
   /SetVariables(Particles.Dark_Glow_Size=&Multiply_Vector2(
     A=&Vector2DFromFloat(Value=80)
     B=&Vector2DFromFloat(Value=&FloatFromCurve(
       FloatCurve=DataInterface() {propertyValues {Curve {keys [<at 0 1.5 RCIM_Cubic>,<at 1 0.5>]}}}
       Scale Curve=&FN_FrontEndFloatModifier(Gameplay=1, Front End=0.6)
      ))
    )) {isSetParameters=true, module=SetVariables_#8}
   &ParticleState()
   &ScaleColor(
    Scale Alpha=&Multiply_Float(
     A=1
     B=&Multiply_Float(
      A=&FNTimeOfDayScalar(Night Scale=1, Morning Scale=0.5, Evening Scale=0.5)
      B=&FloatFromCurve(
       FloatCurve=DataInterface() {propertyValues {Curve {keys [<at 0 0>,<at 0.2 1>,<at 0.8 1>,<at 1 0>]}}}
       Scale Curve=&FNTimeOfDayScalar(Night Scale=1, Morning Scale=0.5, Evening Scale=0.5)
      )
     )
    )
    Scale RGB=<2 2 2>
    ScaleA=true
    ScaleRGB=true
   )
   &ScaleSpriteSize(Scale Factor=&Vector2DFromFloat(Value=&FloatFromCurve(
      FloatCurve=DataInterface() {propertyValues {Curve {keys [<at 0 1.5 RCIM_Cubic>,<at 1 0>]}}}
      Scale Curve=&FN_FrontEndFloatModifier(Gameplay=1, Front End=0.6)
     )))
  ]
  rendererClasses [&NiagaraSpriteRendererProperties]
  renderers [
   {rendererClass=&NiagaraSpriteRendererProperties}
   {rendererClass=&NiagaraSpriteRendererProperties, rendererIndex=1}
  ]
 }
 Flare001 {
  emitterUpdateScript [
   &EmitterState()
   &SpawnBurst_Instantaneous(Spawn Count=1, Use Spawn Probability=false) {enabled=false}
  ]
  particleSpawnScript [
   &InitializeParticle(
    Color=<rgba 0 0 0 1>
    Lifetime=0.25
    Sprite Rotation Mode=::NewEnumerator2
    Sprite Size Mode=::NewEnumerator3
    Uniform Sprite Size=200
    UsePositionOffset=false
    Write Material Random=true
   )
   /SetVariables(Particles.Dark_Glow_CameraOffset=&MakeCustomFloatFromBool(Bool=$User.bIsFrontEndPreview, True Float=1, False Float=50)) {enabled=false, isSetParameters=true, module=SetVariables_#6}
   &CameraOffset(Camera Offset Amount=&MakeCustomFloatFromBool(Bool=$User.bIsFrontEndPreview, True Float=1, False Float=30)) {enabled=false}
  ]
  particleUpdateScript [
   /SetVariables(Particles.Dark_Glow=&MakeLinearColorFromVectorAndFloat(
     "Vector (RGB)"=&MakeVectorFromLinearColorRGB(Color=<rgba 0.0114 0.1123 0.1615 1>)
     "Float (Alpha)"=&Multiply_Float(
      A=&Multiply_Float(A=&FNTimeOfDayScalar(Night Scale=0.7, Morning Scale=0.7, Evening Scale=1), B=0.5)
      B=&FloatFromCurve(FloatCurve=DataInterface() {propertyValues {Curve {keys [<at 0 0>,<at 0.2 1>,<at 0.7 0.6>,<at 1 0>]}}})
     )
    )) {isSetParameters=true, module=SetVariables_#7}
   /SetVariables(Particles.Dark_Glow_Size=&Multiply_Vector2(
     A=&Vector2DFromFloat(Value=80)
     B=&Vector2DFromFloat(Value=&FloatFromCurve(
       FloatCurve=DataInterface() {propertyValues {Curve {keys [<at 0 1.5 RCIM_Cubic>,<at 1 0.5>]}}}
       Scale Curve=&FN_FrontEndFloatModifier(Gameplay=1, Front End=0.6)
      ))
    )) {isSetParameters=true, module=SetVariables_#8}
   &ParticleState()
   &ScaleColor(
    Scale Alpha=&Multiply_Float(
     A=1
     B=&Multiply_Float(
      A=&FNTimeOfDayScalar(Night Scale=1, Morning Scale=0.5, Evening Scale=0.5)
      B=&FloatFromCurve(
       FloatCurve=DataInterface() {propertyValues {Curve {keys [<at 0 0>,<at 0.2 1>,<at 0.8 1>,<at 1 0>]}}}
       Scale Curve=&FNTimeOfDayScalar(Night Scale=1, Morning Scale=0.5, Evening Scale=0.5)
      )
     )
    )
    Scale RGB=<2 2 2>
    ScaleA=true
    ScaleRGB=true
   )
   &ScaleSpriteSize(Scale Factor=&Vector2DFromFloat(Value=&FloatFromCurve(
      FloatCurve=DataInterface() {propertyValues {Curve {keys [<at 0 1.5 RCIM_Cubic>,<at 1 0>]}}}
      Scale Curve=&FN_FrontEndFloatModifier(Gameplay=1, Front End=0.6)
     )))
  ]
  rendererClasses [&NiagaraSpriteRendererProperties]
  renderers [
   {rendererClass=&NiagaraSpriteRendererProperties}
   {rendererClass=&NiagaraSpriteRendererProperties, rendererIndex=1}
  ]
 }
 GPU_particles001 {
  simTarget=GPUComputeSim
  emitterUpdateScript [
   &EmitterState()
   &SpawnRate(SpawnRate=&Lerp_Float(A=0, B=40, Alpha=$System.Timeline.NormalizedTime), Use Spawn Probability=false)
  ]
  particleSpawnScript [
   &InitializeParticle(
    Color=&MakeCustomLinearColorFromBool(Bool=&RandomBool(), True Color=<rgba 0 3.1667 5 1>, False Color=<rgba 0 2.1139 5 1>)
    Lifetime Max=1.5
    Lifetime Mode=::NewEnumerator1
    Sprite Rotation Mode=::NewEnumerator2
    Sprite Size Mode=::NewEnumerator1
    Uniform Sprite Size Max=14
    Uniform Sprite Size Min=6
    UsePositionOffset=false
    Write Material Random=true
   )
   &ShapeLocation(
    Hemisphere Distribution=<1 0.5>
    MaskSpawnGroup=false
    Offset Mode=::NewEnumerator0
    Override Seed=false
    Sphere Surface Distribution=1
    UseOwnerScale=true
   )
   &AddVelocity(
    Rotation Coordinate Space=::World
    Velocity=&RandomRangeVector(Minimum=<-20 -20 200>, Maximum=<20 20 400>)
   )
  ]
  particleUpdateScript [
   /SetVariables(
    Particles.Initial.Color:LinearColor=&MakeCustomLinearColorFromBool(Bool=&RandomBool(), True Color=<rgba 0 5 5 1>, False Color=<rgba 0.359 5 4.3008 1>)
    Particles.Initial.SpriteSize:Vector2f=&Vector2DFromFloat(Value=&RandomRangeFloat(Minimum=6, Maximum=14))
   ) {enabled=false, isSetParameters=true}
   &ParticleState()
   &CurlNoiseForce(Noise Frequency=30, Noise Strength=150, Pan Noise=false)
   &Drag(
    Drag=&RandomRangeFloat(Minimum=3, Maximum=4)
    Rotational Drag=1.5
    Use Linear Drag=true
    Use Rotational Drag=true
   )
   &SolveForcesAndVelocity(Clamp Velocity=false, Limit Acceleration=false)
   &ScaleColor(
    Scale RGB=&Multiply_VectorByFloat(
     Vector=&VectorFromFloat(Value=&FNTimeOfDayScalar(Night Scale=0.75, Morning Scale=1, Evening Scale=1))
     Float=1
    )
    ScaleA=true
    ScaleRGB=true
   )
   &ScaleSpriteSize(Uniform Curve Sprite Scale=DataInterface() {propertyValues {Curve {keys [
        {interpMode=RCIM_Cubic, tangentMode=RCTM_User, time=0, value=1}
        {arriveTangent=-1.9448, interpMode=RCIM_Cubic, leaveTangent=-1.9448, tangentMode=RCTM_User, time=1, value=0}
       ]}}})
   /SetVariables(Particles.Glow Color=&MakeLinearColorFromVectorAndFloat(
     "Vector (RGB)"=&MakeVectorFromLinearColorRGB(Color=&MultiplyLinearColor(Linear Color=<rgba 0 0.54 1 1>, Scale Factor=&MakeVector4_FromFloat(Float=1)))
     "Float (Alpha)"=&Multiply_Float(A=1, B=0.05)
    )) {isSetParameters=true}
   /SetVariables(Particles.Glow size:Vector2f=&Vector2DFromFloat(Value=&Multiply_Float(A=10, B=&FloatFromCurve(FloatCurve=DataInterface() {propertyValues {Curve {keys [<at 0 1>,<at 1 0>]}}})))) {isSetParameters=true}
  ]
  rendererClasses [&NiagaraSpriteRendererProperties]
  renderers [
   {rendererClass=&NiagaraSpriteRendererProperties}
   {rendererClass=&NiagaraSpriteRendererProperties, rendererIndex=1}
  ]
 }
 Smoke {
  emitterUpdateScript [
   &EmitterState()
   &SpawnBurst_Instantaneous(
    Spawn Count=&RandomRangeInt(Minimum=12, Maximum=16)
    Spawn Time=1.125
    Use Spawn Probability=false
    UseLoopCountLimit=true
   ) {enabled=false}
   /SetVariables(
    Emitter.LitColor:LinearColor=&MultiplyLinearColor(
     Linear Color=&MakeLinearColorFromVectorAndFloat(
      "Vector (RGB)"=&VectorFromFloat(Value=&FloatFromCurve(
        FloatCurve=DataInterface() {propertyValues {Curve {keys [
            {tangentMode=RCTM_User, time=0, value=1}
            {arriveTangent=15, interpMode=RCIM_Cubic, leaveTangent=15, time=0.5, value=1}
            {interpMode=RCIM_Cubic, leaveTangent=-55.8892, tangentMode=RCTM_Break, time=0.6, value=10}
            <at 1 1 RCIM_Cubic>
           ]}}}
        CurveIndex=$Emitter.Age
       ))
      "Float (Alpha)"=&Lerp_Float(A=1, B=0.35, Alpha=$System.Dive)
     )
     Scale Factor=<2.2134 15.9838 20 1>
    )
    Emitter.LitPosition:NiagaraPosition=<0 21 -20>
   ) {isSetParameters=true}
  ]
  particleSpawnScript [&InitializeParticle(
    Color=<rgba 0.0156 0.0156 0.0156 1>
    Lifetime Max=1.2
    Lifetime Min=0.8
    Lifetime Mode=::NewEnumerator1
    Sprite Rotation Mode=::NewEnumerator2
    Sprite Size Mode=::NewEnumerator1
    Uniform Sprite Size Max=70
    Uniform Sprite Size Min=50
    UsePositionOffset=false
    Write Material Random=true
   )]
  particleUpdateScript [
   /SetVariables(Particles.Initial.SpriteSize:Vector2f=&Vector2DFromFloat(Value=&RandomRangeFloat(Minimum=70, Maximum=90))) {isSetParameters=true}
   &ParticleState()
   &Drag(Drag=&RandomRangeFloat(Minimum=3, Maximum=4), Use Linear Drag=true, Use Rotational Drag=true)
   &SolveForcesAndVelocity(Clamp Velocity=false, Limit Acceleration=false)
   &ScaleColor(
    Scale Alpha=&FloatFromCurve(
     FloatCurve=DataInterface() {propertyValues {Curve {keys [
         <at 0 0>
         {arriveTangent=0.3312, time=0.1, value=1}
         {arriveTangent=-1.1111, interpMode=RCIM_Cubic, leaveTangent=-1.1111, time=0.7, value=1}
         {arriveTangent=-0.5142, interpMode=RCIM_Cubic, leaveTangent=-0.5142, tangentMode=RCTM_User, time=1, value=0}
        ]}}}
     Scale Curve=&Multiply_Float(
      A=&RandomRangeFloat(Minimum=1, Maximum=1)
      B=&FNTimeOfDayScalar(Night Scale=1, Morning Scale=1, Evening Scale=1)
     )
    )
    Scale RGB=&MakeVectorFromLinearColorRGB(Color=&ColorFromCurve(ColorCurve=DataInterface() {propertyValues {
        AlphaCurve {keys [<at 0 1 RCIM_Cubic>]}
        BlueCurve {keys [<at 0.0637 0.5>, <at 0.2459 2>, <at 0.4736 0.4115>, {interpMode=RCIM_Cubic, leaveTangent=0.7455, time=0.7073, value=0.026}]}
        GreenCurve {keys [<at 0.0637 0.4833>, <at 0.2459 1.1667>, <at 0.4736 0.3977>, {interpMode=RCIM_Cubic, leaveTangent=0.7207, time=0.7073, value=0.023}]}
        RedCurve {keys [<at 0.0637 0>,<at 0.2459 0>,<at 0.4736 0>,<at 0.7073 0 RCIM_Cubic>]}
       }}))
    ScaleA=true
    ScaleRGB=true
   )
   &ScaleSpriteSize(Uniform Curve Sprite Scale=DataInterface() {propertyValues {Curve {keys [
        {arriveTangent=1.8981, interpMode=RCIM_Cubic, leaveTangent=1.8981, tangentMode=RCTM_User, time=0, value=0.35}
        {arriveTangent=0.0897, interpMode=RCIM_Cubic, leaveTangent=0.0897, tangentMode=RCTM_User, time=1, value=1}
       ]}}})
   &FN_SubUVAnimation(End Frame=&Add_Integer(A=$Particles.RandomFrame, B=63), Start Frame=$Particles.RandomFrame, SubUV Loop Count=0.75)
  ]
  rendererClasses [&NiagaraSpriteRendererProperties, &NiagaraLightRendererProperties]
  renderers [
   {rendererClass=&NiagaraSpriteRendererProperties}
   {rendererClass=&NiagaraLightRendererProperties, rendererIndex=1}
  ]
 }
 Smoke_Tail {
  emitterUpdateScript [
   &EmitterState(Life Cycle Mode=::NewEnumerator1, Loop Behavior=::NewEnumerator1, Loop Delay=0.5, Loop Duration Mode=::NewEnumerator1, UseLoopDelay=true)
   &SpawnBurst_Instantaneous(Spawn Count=2, Use Spawn Probability=false, UseLoopCountLimit=false)
  ]
  particleSpawnScript [
   &PartitionParticles()
   &InitializeParticle(
    Color=<rgba 0.125 0.125 0.125 0.5>
    Lifetime=2
    Mesh Scale=&FN_FrontEndVector3Modifier(Gameplay=<1 1 0.5>, Front End=<1 1 0.8>)
    Mesh Scale Mode=::NewEnumerator4
    Sprite Size Mode=::NewEnumerator3
    Uniform Sprite Size=50
    UsePositionOffset=false
    Write Material Random=true
   )
   &InitialMeshOrientation(
    Mesh Axis To Orient=<0 0 0>
    Mesh Orientation Mode=::NewEnumerator2
    Rotation=&Lerp_Vector(A=<0 0 0>, B=<0 0 0>, Alpha=&MakeFloatFromInt(Int=$Particles.PartitionParticles.Partition))
    Rotation Coordinate Space=::NewEnumerator0
    Use Rotation Vector=true
   )
  ]
  particleUpdateScript [
   &ParticleState(
    Kill Particles When Lifetime Has Elapsed=false
    Let Infinitely Lived Particles Die When Emitter Deactivates=true
    Loop Particles Lifetime=true
   )
   &SkeletalMeshLocation(
    Bone / Socket Index=&Add_Integer(A=$Particles.PartitionParticles.Partition, B=1)
    Bone Sampling Mode=::NewEnumerator4
    MaskSpawnGroup=false
    Orientation Sampling=::NewEnumerator1
    Sampled Position Offset=&Lerp_Vector(A=<10 0 0>, B=<-10 0 0>, Alpha=&MakeFloatFromInt(Int=$Particles.PartitionParticles.Partition))
    Skeletal Mesh=$User.SkeletalMesh
   )
   /SetVariables(Particles.Position=&Add_Vector(
     A=&ConvertPositionToVector(Input Position=$Particles.Position, Convert To Absolute Large World Vector=false)
     B=&Add_Vector(A=&Multiply_VectorByFloat(Vector=$System.DiveVelocityScale, Float=10), B=<0 0 0>)
    )) {isSetParameters=true}
   &UpdateMeshOrientation(
    Facing Direction=$System.DiveVelocityScale
    Facing Interpolation=false
    Orientation Method=::NewEnumerator1
    Override Quat=false
    Stabilization Mode:ENiagara_OrientVectorCount=::NewEnumerator2
    TransformFacing=false
    TransformSide=false
   ) {module=UpdateMeshOrientation001}
   &UpdateMeshOrientation(Rotation Rate=-1, Rotation Vector=<0 0 1>) {enabled=false}
   &CameraOffset(Camera Offset Amount=8)
   &ScaleMeshSize(Scale Factor=&Multiply_Vector(
     A=&Lerp_Vector(A=<1 1 1.8>, B=<1 1 2>, Alpha=$System.Timeline.NormalizedTime)
     B=&MakeVector(
      X=1
      Y=1
      Z=&FloatFromCurve(
       FloatCurve=DataInterface() {propertyValues {Curve {keys [<at 0 0 RCIM_Cubic>, {arriveTangent=4.1531, interpMode=RCIM_Cubic, leaveTangent=4.1531, tangentMode=RCTM_User, time=0.5, value=1}]}}}
       CurveIndex=$Emitter.Age
      )
     )
    ))
   &ScaleColor(
    Scale Alpha=&FloatFromCurve(
     FloatCurve=DataInterface() {propertyValues {Curve {keys [<at 0.5 0 RCIM_Cubic>, {arriveTangent=4.1531, interpMode=RCIM_Cubic, leaveTangent=4.1531, tangentMode=RCTM_User, time=1, value=1}]}}}
     CurveIndex=$Emitter.Age
     Scale Curve=0.5
    )
    Scale RGB=&Multiply_VectorByFloat(Vector=&MakeVectorFromLinearColorRGB(Color=<rgba 0.45 0.4 0.5 1>), Float=1)
    ScaleA=true
    ScaleRGB=true
   ) {module=ScaleColor001}
   /SetVariables(Particles.Color_Yellow:LinearColor=&MakeLinearColorFromVectorAndFloat(
     "Vector (RGB)"=&Multiply_VectorByFloat(Vector=&MakeVectorFromLinearColorRGB(Color=<rgba 0.1465 0.7158 1 1>), Float=1)
     "Float (Alpha)"=&FloatFromCurve(
      FloatCurve=DataInterface() {propertyValues {Curve {keys [<at 0 0 RCIM_Cubic>, {arriveTangent=2.7731, interpMode=RCIM_Cubic, leaveTangent=2.7731, tangentMode=RCTM_User, time=1, value=1}]}}}
      CurveIndex=$Emitter.Age
      Scale Curve=0.5
     )
    )) {isSetParameters=true}
   /SetVariables(Particles.Scale_Yellow:Vector3f=&Multiply_Vector(A=$Particles.Scale, B=<0.75 0.75 0.8>)) {isSetParameters=true}
   &ScaleColor(
    Scale Alpha=&FloatFromCurve(
     FloatCurve=DataInterface() {propertyValues {Curve {keys [{interpMode=RCIM_Cubic, tangentMode=RCTM_User, time=0, value=1}, {interpMode=RCIM_Cubic, tangentMode=RCTM_User, time=1, value=1}]}}}
     CurveIndex=$Emitter.Age
    )
    Scale RGB=&MakeVectorFromLinearColorRGB(Color=&LerpLinearColors(
      StartColor=$User.ColorA
      EndColor=$User.ColorB
      LerpFactor=&Lerp_Float(
       A=0
       B=3
       Alpha=&FloatFromCurve(
        FloatCurve=DataInterface() {propertyValues {Curve {keys [
            <at 0 0 RCIM_Cubic>
            {arriveTangent=1.304, interpMode=RCIM_Cubic, leaveTangent=1.304, time=0.1, value=0.75}
            {arriveTangent=-1.2483, interpMode=RCIM_Cubic, leaveTangent=-1.2483, time=0.1681, value=0.2192}
            {arriveTangent=-0.9034, interpMode=RCIM_Cubic, leaveTangent=-0.9034, time=0.216, value=1}
            {arriveTangent=-0.1111, interpMode=RCIM_Cubic, leaveTangent=-0.1111, time=0.3, value=0.1}
            {arriveTangent=1.8356, interpMode=RCIM_Cubic, leaveTangent=1.8356, time=0.3516, value=0.6401}
            {arriveTangent=2.4257, interpMode=RCIM_Cubic, leaveTangent=2.4257, time=0.4219, value=0.05}
            {arriveTangent=0.2031, interpMode=RCIM_Cubic, leaveTangent=0.2031, time=0.5, value=1}
            {arriveTangent=-2.3742, interpMode=RCIM_Cubic, leaveTangent=-2.3742, time=0.5703, value=0.3538}
            {arriveTangent=-1.957, interpMode=RCIM_Cubic, leaveTangent=-1.957, time=0.6325, value=0.6853}
            {arriveTangent=-0.2171, interpMode=RCIM_Cubic, leaveTangent=-0.2171, time=0.7, value=0}
            {arriveTangent=1.2231, interpMode=RCIM_Cubic, leaveTangent=1.2231, time=0.7714, value=0.6552}
            {arriveTangent=1.1257, interpMode=RCIM_Cubic, leaveTangent=1.1257, time=0.8336, value=0.2634}
            {arriveTangent=-1.5833, interpMode=RCIM_Cubic, leaveTangent=-1.5833, time=0.9, value=1}
            <at 1 0 RCIM_Cubic>
           ]}}}
        CurveIndex=$System.NormalizedLoopAge
        Scale Curve=&Lerp_Float(A=0.1, B=1, Alpha=$System.FN_Integrator001.Value)
       )
      )
     ))
    ScaleA=true
    ScaleRGB=true
   )
   &InterpolateOverTime(
    Rate Of Change=10
    Target Vector Value:Vector3f=&ConvertPositionToVector(Input Position=$Particles.Position, Convert To Absolute Large World Vector=false)
    Variable Type=::NewEnumerator2
   ) {enabled=false}
   /SetVariables(Particles.InterpolateDynamic:Vector4f=&MakeFloat4FromVectorAndFloat("Vector (RGB)"=$Output.InterpolateOverTime.Moving Average, "Float (Alpha)"=1)) {enabled=false, isSetParameters=true}
   &DynamicMaterialParameters(
    Index 0 Param 2=&RandomRangeFloat(Minimum=0, Maximum=1.5)
    Index 2 Param 3=0.1
    Index 2 Param 4=1.75
    Module.Param0WriteEnabled=true
    Module.Param0WriteEnabled2=false
    Module.Param1WriteEnabled=true
    Module.Param1WriteEnabled2=false
    Module.Param2WriteEnabled=false
    Module.Param2WriteEnabled2=true
    Module.Param3WriteEnabled=true
    Module.Param3WriteEnabled2=true
    Write Parameter Index 2=true
   )
  ]
  rendererClasses [&NiagaraMeshRendererProperties]
  renderers [
   {rendererClass=&NiagaraMeshRendererProperties, rendererIndex=0}
   {rendererClass=&NiagaraMeshRendererProperties, rendererIndex=1}
  ]
 }
 Speedlines {
  simTarget=GPUComputeSim
  emitterUpdateScript [
   &EmitterState(Life Cycle Mode=::NewEnumerator1, Loop Duration=1, UseLoopDelay=false)
   &SpawnRate(
    SpawnRate=&Multiply_Float(
     A=$System.Timeline.NormalizedTime
     B=&FN_FrontEndFloatModifier(Gameplay=&UniformRangedFloat(Minimum=100, Maximum=200), Front End=0)
    )
    Use Spawn Probability=false
   )
  ]
  particleSpawnScript [
   /Niagara/Modules/Spawn/Initialization/InitializeParticle(
    Color=&MakeLinearColorFromVectorAndFloat(
     "Vector (RGB)"=<1 1 1>
     "Float (Alpha)"=&Multiply_Float(
      A=&Multiply_Float(
       A=&UniformRangedFloat(Minimum=0.015, Maximum=0.15)
       B=&FN_FrontEndFloatModifier(Gameplay=1, Front End=0.5)
      )
      B=&Add_Float(
       A=1
       B=&Multiply_Float(A=10, B=&Subtract_Float(A=1, B=$Engine.Emitter.SpawnCountScale))
      )
     )
    )
    Lifetime=&FN_FrontEndFloatModifier(
     Gameplay=&UniformRangedFloat(Minimum=0.15, Maximum=0.3)
     Front End=&UniformRangedFloat(Minimum=0.1, Maximum=0.2)
    )
    Module.Write Color=true
    Module.Write Lifetime=true
    Module.Write Mass=true
    Module.Write Material Random=true
    Module.Write Position=false
    Module.Write Scale=false
    Module.Write SpriteRotation=false
    Module.Write SpriteSize=true
    Position=/Niagara/Functions/Localspace/SimulationPosition()
    Sprite Size=&UniformRangedVector2D(Minimum=<3 4>, Maximum=<5 6>)
   ) {enabled=false}
   /Niagara/Modules/Spawn/Initialization/V2/InitializeParticle(
    Color=&MakeLinearColorFromVectorAndFloat(
     "Vector (RGB)"=<1 1 1>
     "Float (Alpha)"=&Multiply_Float(
      A=&Multiply_Float(
       A=&UniformRangedFloat(Minimum=0.015, Maximum=0.15)
       B=&FN_FrontEndFloatModifier(Gameplay=1, Front End=0.5)
      )
      B=&Add_Float(
       A=1
       B=&Multiply_Float(A=10, B=&Subtract_Float(A=1, B=$Engine.Emitter.SpawnCountScale))
      )
     )
    )
    Lifetime=&FN_FrontEndFloatModifier(
     Gameplay=&UniformRangedFloat(Minimum=0.15, Maximum=0.3)
     Front End=&UniformRangedFloat(Minimum=0.1, Maximum=0.2)
    )
    Mass Mode=::NewEnumerator1
    Sprite Size=&UniformRangedVector2D(Minimum=<3 4>, Maximum=<5 6>)
    Sprite Size Mode=::NewEnumerator4
    UsePositionOffset=false
    Write Material Random=true
   ) {module=InitializeParticle001}
   &CylinderLocation(
    Cylinder Radius=&UniformRangedFloat(Minimum=100, Maximum=150)
    MaskSpawnGroup=false
    Override Seed=false
    UseOwnerScale=true
   )
   &AddVelocity(Velocity=&FN_FrontEndVector3Modifier(Gameplay=<0 0 1000>, Front End=<0 0 3000>))
   &AddVelocityFromPoint(
    Module.Use Falloff Distance=false
    Velocity Strength=&Multiply_Float(
     A=&FN_FrontEndFloatModifier(Gameplay=1, Front End=2)
     B=&UniformRangedFloat(Minimum=500, Maximum=700)
    )
   )
  ]
  particleUpdateScript [
   &ParticleState()
   &SolveForcesAndVelocity(Clamp Velocity=false, Limit Acceleration=false)
   &ScaleSpriteSize(Scale Factor=&Vector2DFromCurve(
     Vector2Curve=DataInterface() {propertyValues {XCurve {keys [<at 0 0.5 RCIM_Cubic>,<at 0.5 1 RCIM_Cubic>]}, YCurve {keys [<at 0 0 RCIM_Cubic>,<at 0.5 1 RCIM_Cubic>]}}}
     ScaleCurve=&Vector2DFromFloat(Value=1)
    ))
   &ScaleSpriteSizeBySpeed(Max Scale Factor=<1 40>, Min Scale Factor=<0.01 0.05>, Velocity Threshold=1)
   &ScaleColor(
    Scale Alpha=&FloatFromCurve(
     FloatCurve=DataInterface() {propertyValues {Curve {keys [<at 0 0 RCIM_Cubic>, <at 0.2 1>, {arriveTangent=-2.221, interpMode=RCIM_Cubic, leaveTangent=-2.221, tangentMode=RCTM_User, time=1, value=0}]}}}
     Scale Curve=&FN_FrontEndFloatModifier(Gameplay=&Lerp_Float(A=0, B=0.4, Alpha=$System.Timeline.NormalizedTime), Front End=0)
    )
    ScaleA=true
    ScaleRGB=true
   )
  ]
  rendererClasses [&NiagaraSpriteRendererProperties]
  renderers [{rendererClass=&NiagaraSpriteRendererProperties}]
 }
 Trail_Sparkles {
  emitterUpdateScript [
   &EmitterState(Life Cycle Mode=::NewEnumerator1, Loop Delay=1.15, Loop Duration=1, UseLoopDelay=false)
   &SpawnRate(SpawnRate=30, Use Spawn Probability=false)
  ]
  particleSpawnScript [
   &PartitionParticles()
   &ShapeLocation(MaskSpawnGroup=false, Override Seed=false, Sphere Radius=5, UseOwnerScale=true)
   &InitializeParticle(
    Color=<rgba 0.0067 0.0303 0.0417 1>
    Lifetime Max=1.2
    Lifetime Min=0.8
    Lifetime Mode=::NewEnumerator1
    Sprite Rotation Mode=::NewEnumerator0
    Sprite Size Mode=::NewEnumerator1
    Uniform Sprite Size Max=200
    Uniform Sprite Size Min=200
    UsePositionOffset=false
    Write Material Random=true
   )
   &AddVelocity(
    Cone Angle=5
    Cone Axis=<0 0 1>
    Module.UseConeFalloff=false
    Module.UseVelDistribution=false
    Override Seed=false
    Rotation Coordinate Space=::World
    Velocity Mode=::NewEnumerator2
    Velocity Speed=&Multiply_Float(
     A=&RandomRangeFloat(Minimum=200, Maximum=400)
     B=&Lerp_Float(
      A=1
      B=&FN_FrontEndFloatModifier(Gameplay=2, Front End=1)
      Alpha=&ClampFloat(Float=$User.ParaGlideLeanAlpha)
     )
    )
   )
   /SetVariables(Particles.RandomCore:NiagaraFloat=&FloatFromCurve(
     FloatCurve=DataInterface() {propertyValues {Curve {keys [
         {arriveTangent=0.0559, interpMode=RCIM_Cubic, leaveTangent=0.0559, tangentMode=RCTM_User, time=0, value=0.2}
         {arriveTangent=0.1895, interpMode=RCIM_Cubic, leaveTangent=0.1895, tangentMode=RCTM_User, time=0.9, value=0.25}
         {arriveTangent=1.1808, interpMode=RCIM_Cubic, leaveTangent=1.1808, tangentMode=RCTM_User, time=1, value=0.3}
        ]}}}
     CurveIndex=&RandomRangeFloat(Minimum=0, Maximum=1)
     Scale Curve=2
    )) {enabled=false, isSetParameters=true}
  ]
  particleUpdateScript [
   &SkeletalMeshLocation(
    Bone / Socket Index=&Add_Integer(A=$Particles.PartitionParticles.Partition, B=1)
    Bone Sampling Mode=::NewEnumerator4
    MaskSpawnGroup=false
    Orientation Sampling=::NewEnumerator1
    Position Sampling=::NewEnumerator1
    Sampled Position Offset=$Particles.Initial.ShapeLocation.ShapePosition
    Skeletal Mesh=$User.SkeletalMesh
   )
   /SetVariables(Particles.Glow03_Color=&MakeLinearColorFromVectorAndFloat(
     "Vector (RGB)"=&MakeVectorFromLinearColorRGB(Color=&MultiplyLinearColor(Linear Color=<rgba 0.543 0.1656 1 0.2>, Scale Factor=&MakeVector4_FromFloat(Float=1)))
     "Float (Alpha)"=&Multiply_Float(A=0.3, B=1)
    )) {enabled=false, isSetParameters=true}
   /SetVariables(Particles.Glow03_Size=&Multiply_Vector2D_ByFloat(
     Float=&FloatFromCurve(
      FloatCurve=DataInterface() {propertyValues {Curve {keys [
          <at 0 0 RCIM_Cubic>
          <at 0.2 1.3 RCIM_Cubic>
          {arriveTangent=-0.1423, interpMode=RCIM_Cubic, leaveTangent=-0.1423, tangentMode=RCTM_User, time=1, value=0}
         ]}}}
      Scale Curve=15
     )
     Vector2D=<1 1.5>
    )) {enabled=false, isSetParameters=true}
   &ParticleState()
   &CurlNoiseForce(Noise Strength=&Multiply_Float(A=300, B=1), Pan Noise=true, Pan Noise Field=<0 0 1>) {enabled=false}
   &ScaleSpriteSize(Scale Factor=&Vector2DFromFloat(Value=&FloatFromCurve(
      FloatCurve=DataInterface() {propertyValues {Curve {keys [
          <at 0 0 RCIM_Cubic>
          <at 0.2 1.2 RCIM_Cubic>
          {arriveTangent=-2.7132, interpMode=RCIM_Cubic, leaveTangent=-2.7132, tangentMode=RCTM_User, time=1, value=0}
         ]}}}
      Scale Curve=&FN_FrontEndFloatModifier(Gameplay=0.3, Front End=0.2)
     )))
   &ScaleColor(
    Scale Alpha=&Multiply_Float(A=1, B=1)
    Scale RGB=&Multiply_VectorByFloat(Vector=<2 2 2>, Float=$System.TOD)
    ScaleA=true
    ScaleRGB=true
   )
   &Drag(Drag=2, Use Linear Drag=true, Use Rotational Drag=false)
   /SetVariables(Particles.RandomFrame=&RandomRangeInt(Minimum=0, Maximum=63)) {isSetParameters=true}
   &SolveForcesAndVelocity(Clamp Velocity=false, Limit Acceleration=false)
   &FN_SubUVAnimation(End Frame=&Add_Integer(A=$Particles.RandomFrame, B=63), Start Frame=$Particles.RandomFrame, SubUV Loop Count=0.75)
  ]
  rendererClasses [&NiagaraSpriteRendererProperties]
  renderers [{rendererClass=&NiagaraSpriteRendererProperties}]
 }
}
summary {
 systemName=NS_Contrail_LazarusLens
 emitters [
  {emitterName=Trail_Sparkles, rendererClasses [&NiagaraSpriteRendererProperties]}
  {emitterName=Clock, rendererClasses [&NiagaraMeshRendererProperties]}
  {emitterName=Bats, rendererClasses [&NiagaraMeshRendererProperties, &NiagaraSpriteRendererProperties]}
  {emitterName=Birds, rendererClasses [&NiagaraMeshRendererProperties, &NiagaraSpriteRendererProperties]}
  {emitterName=Color_wave, rendererClasses [&NiagaraSpriteRendererProperties]}
  {emitterName=Smoke, rendererClasses [&NiagaraSpriteRendererProperties, &NiagaraLightRendererProperties]}
  {emitterName=Burst_Air_Meshes, rendererClasses [&NiagaraMeshRendererProperties]}
  {emitterName=Flare, rendererClasses [&NiagaraSpriteRendererProperties]}
  {emitterName=Speedlines, simTarget=GPUComputeSim, rendererClasses [&NiagaraSpriteRendererProperties]}
  {emitterName=Smoke_Tail, rendererClasses [&NiagaraMeshRendererProperties]}
  {emitterName=GPU_particles001, simTarget=GPUComputeSim, rendererClasses [&NiagaraSpriteRendererProperties]}
  {emitterName=Flare001, rendererClasses [&NiagaraSpriteRendererProperties]}
 ]
 userVariables [
  User.bIsFrontEnd {type=NiagaraBool}
  User.bIsFrontEndPreview {type=NiagaraBool}
  User.ColorA {type=LinearColor, description="", defaultValue {struct=LinearColor, value=<rgba 0.5559 0.6633 0.9583 1>}}
  User.ColorB {type=LinearColor, description="", defaultValue {struct=LinearColor, value=<rgba 0.7969 0.0266 0 1>}}
  User.FEVec {type=Vector3f, description="", defaultValue {struct=Vector3f, value=<0 0 0>}}
  User.FX Stage {type=NiagaraFloat, defaultValue=0, description=""}
  User.ParaGlideLeanAlpha {type=NiagaraFloat, defaultValue=0, description=""}
  User.PlayerSpeed {type=NiagaraFloat, defaultValue=0, description=""}
  User.Return {type=NiagaraBool}
  User.RibbonIDOffset {type=NiagaraInt32}
  User.SK_Mesh {
   type=NiagaraDataInterfaceSkeletalMesh
   defaultValue {
    struct=NiagaraExt_VariableValue_DataInterface
    value {dataInterface=&self:NiagaraDataInterfaceSkeletalMesh_1}
   }
  }
  User.SkeletalMesh {
   type=NiagaraDataInterfaceSkeletalMesh
   defaultValue {
    struct=NiagaraExt_VariableValue_DataInterface
    value {dataInterface=&self:NiagaraDataInterfaceSkeletalMesh_0}
   }
  }
  User.Variant_Float {type=NiagaraFloat, defaultValue=0, description=""}
 ]
}
